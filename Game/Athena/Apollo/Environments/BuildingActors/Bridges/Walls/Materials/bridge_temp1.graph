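# bridge_temp1 is a temporary-named Default Lit material in the Apollo environment's bridge wall materials folder.
# Its graph holds a single vector parameter with a muted green default color.
# The file shows no connection from that parameter to any material output, and no textures or other logic.
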Material /Game/Athena/Apollo/Environments/BuildingActors/Bridges/Walls/Materials/bridge_temp1
graph {
 properties {pixelDepthOffsetMode=PDOM_Legacy, usedShadingModels=Default Lit}
 nodes [VectorParameter() {defaultValue=<rgba 0.1915 0.3646 0.1576 0>}]
}
